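# Masked, two-sided Default Lit material for the maximum-LOD version of the industrial power tower 02 prop.
# Its graph holds a texture sample of the dedicated MaxLOD texture and a constant of one, with no connections shown.
# It is flagged for use with Nanite and instanced static meshes.
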Material /Game/Environments/Props/Industrial/Materials/LOD/M_Prop_Industrial_PowerTower_02_MaxLOD
graph {
 properties {
  bUsedWithInstancedStaticMeshes=true
  bUsedWithNanite=true
  blendMode=BLEND_Masked
  pixelDepthOffsetMode=PDOM_Legacy
  twoSided=true
  usedShadingModels=Default Lit
 }
 nodes [
  Constant() {r=1}
  TextureSample() {texture=&/Prop_Industrial_PowerTower_02_MaxLOD}
 ]
}
